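# Parent material for the BullKeynote pickaxe, layering an animated energy effect onto the shared 2023 anime character shader on Cloth shading.
# Time-stepped stardust noise plus scrolling clouds modulated by animated gradient noise index a curve atlas for energy color, with stabilized screen-space UVs.
# A color-distance SphereMask defines an edge mask within the effect, and reroute-driven lerps feed lit, shaded and emissive inputs.
# It also writes temporal responsiveness.
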Material /BRCosmetics/Weapons/FORT_Melee/Pickaxe_BullKeynote/Materials/M_Pickaxe_BullKeynote_Parent
graph {
 properties {
  bAutomaticallySetUsageInEditor=false
  bUseMaterialAttributes=true
  bUsedWithClothing=true
  bUsedWithMeshDeformer=true
  bUsedWithMorphTargets=true
  bUsedWithNiagaraMeshParticles=true
  bUsedWithSkeletalMesh=true
  bUsedWithStaticLighting=true
  pixelDepthOffsetMode=PDOM_Legacy
  previewMesh=/BRCosmetics/Weapons/FORT_Melee/Pickaxe_BullKeynote/Meshes/BullKeynote_Axe
  shadingModel=MSM_Cloth
  usedShadingModels=Cloth
 }
 nodes [
  n1 = MaterialFunctionCall() {materialFunction=&BoundingBoxBased_0-1_UVW}
  n2 = ObjectLocalBounds()
  ComponentMask(None=Multiply(A=@n1.RGB, B=@n2.Extents)) {b=true, r=true}
  ComponentMask(None=@n2.Extents) {b=true, r=true}
  Divide(A=@n1.RGB, B=@n2.Extents)
  n3 = MaterialFunctionCall!(CustomUV) {materialFunction=/Game/Global/Materials/Character_Parent_2023/Anime_Parent_2023/MaterialFunctions/MFC_AnimeShader_2023_Pre}
  n4 = MaterialFunctionCall!(CustomLightVector, OverrideNormal_WS, STT=@n3.STT, Normal_Map=@n3.STT, DFL=@n3.STT, FX=@n3.STT) {materialFunction=/Game/Global/Materials/Character_Parent_2023/Anime_Parent_2023/MaterialFunctions/MFC_AnimeShader_2023_Calculations}
  n5 = NamedRerouteUsage()
  n6 = NamedRerouteUsage()
  MaterialFunctionCall!(
   Shading_Gradient=@n4.Shading_Gradient
   Threshold_Gradient=@n4.Shading_Gradient
   Specular_Gradient=@n4.Shading_Gradient
   Sheen_Mask=@n4.Shading_Gradient
   Thickness=@n4.Shading_Gradient
   DFL_Mask=@n4.Shading_Gradient
   FX=@n4.Shading_Gradient
   Metallic_Gradient=@n4.Shading_Gradient
   RimLight_Gradient=@n4.Shading_Gradient
   NormalBuffer=@n4.Shading_Gradient
   --------
   CINE_LightSources=@n4.Shading_Gradient
   CINE_LightColor=@n4.Shading_Gradient
   ---------
   Color_Lit=LinearInterpolate(A=NamedRerouteUsage(), B=@n3.Color_Lit, Alpha=NamedRerouteUsage())
   Color_Shaded=LinearInterpolate(A=NamedRerouteUsage(), B=@n3.Color_Shaded, Alpha=NamedRerouteUsage())
   SRM=@n3.SRM
   Emissive=LinearInterpolate(A=@n5, B=@n6, Alpha=NamedRerouteUsage())
   DFL_Color=@n3.SRM
   Specular_Color=@n3.SRM
   RimLight_Color=@n3.SRM
  ) {materialFunction=/Game/Global/Materials/Character_Parent_2023/Anime_Parent_2023/MaterialFunctions/MFC_AnimeShader_2023_Post}
  n7 = NamedRerouteDeclaration(None=MaterialFunctionCall!(ObjectPivotLocation, ObjectPositionOffset=Constant3Vector()) {materialFunction=/Game/Characters/Player/Common/MaterialFunctions/Superlevel/ImprovedStabilizedScreenSpaceUVs}.StabilizedUVs) {name=ScreenUV, nodeColor=<rgba 1 0 0.6797 1>}
  n8 = ScalarParameter() {defaultValue=4, parameterName=StardustScale}
  n9 = Constant() {r=3}
  n10 = Multiply(A=Time(), B=@n9) {constB=3}
  n11 = Floor(None=@n10)
  n12 = TextureSample(UVs=Add(A=Multiply(A=@n7, B=@n8) {constB=2048}, B=Divide(A=@n11, B=@n9) {constB=3})) {texture=/BRCosmetics/Weapons/FORT_Melee/Pickaxe_BullKeynote/Textures/T_BullKeyNote_Noise}
  n13 = Frac(None=@n10)
  n14 = Multiply(A=NamedRerouteUsage(), B=ScalarParameter() {defaultValue=4, parameterName=CloudsScale})
  n15 = TextureSample(UVs=Panner(Coordinate=@n14) {speedY=0.076}) {texture=/BRCosmetics/Weapons/FORT_Melee/Pickaxe_BullKeynote/Textures/T_BullKeyNote_Clouds}
  n16 = Multiply(A=Time(), B=ScalarParameter() {defaultValue=0.3, parameterName=CloudsAnimSpeed})
  n17 = Noise!(World Position=AppendVector(A=@n14, B=@n16), FilterWidth) {bTiling=true, noiseFunction=NOISEFUNCTION_GradientTex3D, repeatSize=4}
  n18 = Saturate(None=Power(Base=ConstantBiasScale(None=@n17), Exp=ScalarParameter() {defaultValue=2, parameterName=CloudsExp}))
  n19 = Add(
   A=Power(Base=MaterialFunctionCall(A=@n12.R, B=@n12.R, C=@n12.R, Alpha=@n13) {materialFunction=&Lerp_3Color}.Result) {constExponent=1}
   B=Multiply(A=Multiply(A=@n15.R, B=@n18), B=ScalarParameter() {defaultValue=3, parameterName=CloudsBoost}) {constA=3}
  )
  MaterialFunctionCall(
   A=VectorParameter() {parameterName=ColorA, defaultValue=<rgba 0 0.2947 1 1>}.RGB
   B=VectorParameter() {parameterName=ColorB, defaultValue=<rgba 0 0.9528 1 1>}.RGB
   C=VectorParameter() {parameterName=ColorC, defaultValue=<rgba 0.5938 0.8201 1 1>}.RGB
   Alpha=@n19
  ) {materialFunction=&Lerp_3Color}
  Multiply(A=@n3.Color_Lit, B=@n5)
  Multiply(A=@n3.Color_Shaded, B=@n5)
  Multiply(A=@n6, B=@n5)
  Multiply(
   A=MaterialFunctionCall!(Vector=CameraVectorWS()) {materialFunction=/Engine/Functions/Engine_MaterialFunctions01/Coordinates/LongLatToUV}.Result
   B=Constant2Vector() {g=1, r=2}
  )
  n20 = Multiply(A=NamedRerouteUsage(), B=ScalarParameter() {defaultValue=4, parameterName=CloudsScale})
  n21 = Panner(Coordinate=Multiply(A=@n20) {constB=1.1071}, Time=@n16) {speedX=0.1, speedY=-0.03}
  n22 = Panner(Coordinate=Multiply(A=@n20) {constB=1.3175}, Time=@n16) {speedX=-0.246, speedY=-0.112}
  Multiply(
   A=TextureSample(UVs=Panner(Coordinate=@n20, Time=@n16) {speedY=0.3}) {samplerType=SAMPLERTYPE_Grayscale, texture=/Game/Global/Textures/Default/Noise/LowResBlurredNoiseNormalized}.R
   B=Multiply(
    A=TextureSample(UVs=@n21) {samplerType=SAMPLERTYPE_Grayscale, texture=&LowResBlurredNoiseNormalized}.R
    B=TextureSample(UVs=@n22) {samplerType=SAMPLERTYPE_Grayscale, texture=&LowResBlurredNoiseNormalized}.R
   )
  )
  n23 = CurveAtlasRowParameter(CurveTime=@n19) {
   atlas=&/CA_BullKeynote
   curve=&/C_BullKeynote_Energy
   parameterName=EnergyColorCurve
  }
  n24 = LinearInterpolate(A=ScalarParameter() {defaultValue=0.03, parameterName=MinBrightness}, Alpha=@n19) {constA=0.03}
  n25 = SphereMask(
   A=NamedRerouteUsage()
   B=VectorParameter() {parameterName=EmissiveTexMaskColor, defaultValue=<rgba 1 1 1 1>}.RGB
   Radius=ScalarParameter() {defaultValue=0.2, parameterName=EmissiveTexMaskRadius}
   Hardness=ScalarParameter() {defaultValue=0.5, parameterName=EmissiveTexMaskHardness}
  ) {attenuationRadius=0.2, hardnessPercent=0.5}
  NamedRerouteDeclaration(None=LinearInterpolate(A=Multiply(A=@n23, B=@n24), Alpha=@n25)) {name=EnergyNoiseEffect, nodeColor=<rgba 1 0 0.1172 1>}
  NamedRerouteDeclaration(None=@n3.Emissive) {name=Emissive_IN, nodeColor=<rgba 1 0.2812 0 1>}
  NamedRerouteDeclaration(None=@n25) {name=EdgeMask, nodeColor=<rgba 0 1 0.2734 1>}
  Power(Base=@n19) {constExponent=1}
  TemporalResponsivenessOutput!(None=LinearInterpolate(A=Constant() {r=0.25}, Alpha=NamedRerouteUsage()) {constB=0.25})
  TextureCoordinate() {uTiling=0.5, vTiling=0.5}
 ]
}
